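# Parent material for the TackleShow pickaxe, layering an animated glowing trail effect over the shared 2023 anime character shader with Cloth shading.
# On UV1, two panned trail textures, one with warped UVs, multiply and join sphere-masked scrolling noise, then get faded and tinted.
# An edge mask blends inner and edge colors into that glow.
# An FX texture mask swaps the glow into emissive and drives temporal responsiveness.
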Material /BRCosmetics/Weapons/FORT_Melee/Pickaxe_TackleShow/Materials/M_Pickaxe_TackleShow_Parent
graph {
 properties {
  bAutomaticallySetUsageInEditor=false
  bHasPixelAnimation=true
  bUseMaterialAttributes=true
  bUsedWithClothing=true
  bUsedWithMeshDeformer=true
  bUsedWithMorphTargets=true
  bUsedWithNiagaraMeshParticles=true
  bUsedWithSkeletalMesh=true
  bUsedWithStaticLighting=true
  pixelDepthOffsetMode=PDOM_Legacy
  previewMesh=/BRCosmetics/Weapons/FORT_Melee/Pickaxe_TackleShow/DEMO/TackleShowDEMO
  shadingModel=MSM_Cloth
  usedShadingModels=Cloth
 }
 nodes [
  n1 = NamedRerouteUsage()
  Multiply(A=@n1) {constB=3}
  n2 = TextureCoordinate() {coordinateIndex=1}
  NamedRerouteDeclaration(None=OneMinus(None=TextureSample(UVs=@n2) {texture=/BRCosmetics/Weapons/FORT_Melee/Pickaxe_TackleShow/Textures/T_TackleShow_Pickaxe_FX1}.R)) {name=EdgeMask, nodeColor=<rgba 0.7812 1 0 1>}
  n3 = TextureCoordinate() {coordinateIndex=1}
  n4 = ComponentMask(None=@n3) {r=true}
  n5 = OneMinus(None=@n4)
  n6 = OneMinus(None=Multiply(A=Multiply(A=@n4, B=@n5), B=ScalarParameter() {defaultValue=3.8004, parameterName=s}) {constB=12})
  n7 = ComponentMask(None=@n3) {g=true}
  n8 = OneMinus(None=@n7)
  n9 = ScalarParameter() {defaultValue=-2.8748, parameterName=Param_1}
  n10 = ScalarParameter() {defaultValue=70.4997}
  n11 = OneMinus(None=Multiply(A=Multiply(A=@n7, B=@n8), B=LinearInterpolate(A=@n9, B=@n10, Alpha=@n7)) {constB=12})
  Saturate(None=Add(A=Saturate(None=@n6), B=Saturate(None=@n11)))
  n12 = MaterialFunctionCall!(CustomUV) {materialFunction=/Game/Global/Materials/Character_Parent_2023/Anime_Parent_2023/MaterialFunctions/MFC_AnimeShader_2023_Pre}
  n13 = MaterialFunctionCall!(CustomLightVector, OverrideNormal_WS, STT=@n12.STT, Normal_Map=@n12.STT, DFL=@n12.STT, FX=@n12.STT) {materialFunction=/Game/Global/Materials/Character_Parent_2023/Anime_Parent_2023/MaterialFunctions/MFC_AnimeShader_2023_Calculations}
  n14 = MaterialFunctionCall!(
   Shading_Gradient=@n13.Shading_Gradient
   Threshold_Gradient=@n13.Shading_Gradient
   Specular_Gradient=@n13.Shading_Gradient
   Sheen_Mask=@n13.Shading_Gradient
   Thickness=@n13.Shading_Gradient
   DFL_Mask=@n13.Shading_Gradient
   FX=@n13.Shading_Gradient
   Metallic_Gradient=@n13.Shading_Gradient
   RimLight_Gradient=@n13.Shading_Gradient
   NormalBuffer=@n13.Shading_Gradient
   --------
   CINE_LightSources=@n13.Shading_Gradient
   CINE_LightColor=@n13.Shading_Gradient
   ---------
   Color_Lit=@n12.Color_Lit
   Color_Shaded=@n12.Color_Lit
   SRM=@n12.Color_Lit
   Emissive=@n12.Color_Lit
   DFL_Color=@n12.Color_Lit
   Specular_Color=@n12.Color_Lit
   RimLight_Color=@n12.Color_Lit
  ) {materialFunction=/Game/Global/Materials/Character_Parent_2023/Anime_Parent_2023/MaterialFunctions/MFC_AnimeShader_2023_Post}
  n15 = TextureSample() {texture=/BRCosmetics/Weapons/FORT_Melee/Pickaxe_TackleShow/Textures/T_TackleShow_Pickaxe_FX}
  n16 = TextureSampleParameter2D(UVs=Panner(Coordinate=TextureCoordinate() {coordinateIndex=1, uTiling=0.25, vTiling=0.5}) {speedY=1}) {parameterName=Noise, samplerType=SAMPLERTYPE_LinearGrayscale, texture=/Game/Effects/Fort_Effects/Textures/T_DarkAndLight_Noise}
  n17 = TextureCoordinate() {coordinateIndex=1}
  n18 = ComponentMask(None=@n17) {g=true}
  n19 = TextureCoordinate() {coordinateIndex=1, uTiling=0.5}
  n20 = Constant2Vector() {g=0.95, r=0.25}
  n21 = Multiply(A=SmoothStep(Value=@n18) {constMin=0.6}, B=Multiply(A=SphereMask(A=@n19, B=@n20) {attenuationRadius=0.2, hardnessPercent=1}) {constB=2})
  n22 = OneMinus(None=@n1)
  n23 = AppendVector(A=@n18, B=ComponentMask(None=@n17) {r=true})
  n24 = Constant2Vector() {g=5, r=1}
  n25 = LinearInterpolate(A=Add(A=Multiply(A=@n23, B=@n24), B=Constant2Vector() {g=-2.2086}), B=@n23, Alpha=Saturate(None=Multiply(A=@n18) {constB=3.5}))
  n26 = Multiply(
   A=TextureSampleParameter2D(UVs=Panner(Coordinate=@n25) {speedX=0.75}) {parameterName=Trail, texture=/BRCosmetics/Weapons/FORT_Melee/Pickaxe_TackleShow/Textures/T_Pickaxe_TackleShow_Blur_Trail}.R
   B=TextureSampleParameter2D(UVs=Panner(Coordinate=@n23) {speedX=0.35}) {parameterName=Trail2, texture=&T_Pickaxe_TackleShow_Blur_Trail}.R
  )
  n27 = Saturate(None=Add(A=Multiply(A=@n16.R, B=@n21), B=Multiply(A=@n22, B=@n26)))
  n28 = OneMinus(None=@n18)
  n29 = Saturate(None=Multiply(A=Power(Base=@n28) {constExponent=13}) {constB=5})
  n30 = Multiply(A=VectorParameter() {parameterName=TrailColor, defaultValue=<rgba 0.4508 1 1 1>}.RGB, B=Saturate(None=Subtract(A=@n27, B=@n29)))
  n31 = Multiply(A=LinearInterpolate(
    A=VectorParameter() {parameterName=InnerColor, defaultValue=<rgba 0.227 1 1 1>}.RGB
    B=VectorParameter() {parameterName=Edge Color, defaultValue=<rgba 0 0.5841 0.2961 1>}.RGB
    Alpha=NamedRerouteUsage()
   )) {constB=1.5}
  SetMaterialAttributes(
   MaterialAttributes=@n14.Result
   Emissive Color=LinearInterpolate(
    A=GetMaterialAttributes(None=@n14.Result) {attributeGetTypes [#1]}.Emissive Color
    B=Add(A=@n30, B=@n31)
    Alpha=@n15.R
   )
  ) {attributeSetTypes [#1]}
  TemporalResponsivenessOutput!(None=LinearInterpolate(Alpha=@n15.R) {constB=0.5})
 ]
}
